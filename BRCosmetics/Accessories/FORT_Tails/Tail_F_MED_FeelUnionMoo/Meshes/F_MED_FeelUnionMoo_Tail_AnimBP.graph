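# Animation Blueprint for the FeelUnionMoo tail cosmetic on the shared Fortnite tail skeleton.
# Blends the input pose with StaminaCat's tucked-tail sequence by bool, then layers chained Modify Bone nodes that offset the root and bend tail joints, using bool alphas with varied blend times.
# On customization load, a long backbling counts as a cape and hides a material element via dynamic material parameter.
# Animation updates derive crouch-only and not-moving flags.
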AnimBlueprint /BRCosmetics/Accessories/FORT_Tails/Tail_F_MED_FeelUnionMoo/Meshes/F_MED_FeelUnionMoo_Tail_AnimBP
target_skeleton=/Game/Accessories/FORT_Tails/Common/Fortnite_Base_Tail_Skeleton
graphs [
 AnimGraph {nodes [
   n1 = Input Pose()
   n2 = Random Sequence Player!() -> (Pose:Pose Link Structure) {
    is=title AnimGraphNode_RandomPlayer
    entries [{
      chanceToPlay=1
      maxLoopCount=0
      maxPlayRate=1
      minLoopCount=0
      minPlayRate=1
      sequence=/Game/Accessories/FORT_Tails/Tail_M_MED_StaminaCat/Animations/StaminaCat_Tucked_Tail
      blendIn {blendOption=Linear, blendTime=0.2, customCurve=null}
     }]
   }
   n3 = "Transform (Modify) Bone - Bone: C_Tail_A_Base_Jnt"!(
    ComponentPose=Local To Component(LocalPose=Blend Poses by bool(BlendPose_0=@n2, BlendPose_1=@n1) {blendType=Linear})
    Translation="0,0,0"
    Rotation="13,0,0"
   ) -> (Pose) {
    alphaInputType=Bool
    rotationMode=BMM_Additive
    rotationSpace=BCS_BoneSpace
    scaleSpace=BCS_BoneSpace
    translationMode=BMM_Replace
    translationSpace=BCS_BoneSpace
    alphaBoolBlend {blendInTime=1, blendOption=Linear, blendOutTime=1}
    boneToModify {boneName=C_Tail_A_Base_Jnt}
    rotation {pitch=13}
   }
   n4 = "Transform (Modify) Bone - Bone: C_Root_Main_Root_Jnt"!(ComponentPose=@n3, bAlphaBoolEnabled:bool=True, Translation="0,10,0", Rotation="0,0,8") -> (Pose) {
    alphaInputType=Bool
    rotationMode=BMM_Replace
    rotationSpace=BCS_ParentBoneSpace
    translationMode=BMM_Replace
    translationSpace=BCS_ParentBoneSpace
    boneToModify {boneName=C_Root_Main_Root_Jnt}
    rotation {roll=8}
    translation {y=10}
   }
   n5 = "Transform (Modify) Bone - Bone: C_Tail_A_1_Jnt"!(ComponentPose=@n4, bAlphaBoolEnabled:bool=True, Rotation="15,0,0") -> (Pose) {
    alphaInputType=Bool
    rotationMode=BMM_Additive
    rotationSpace=BCS_BoneSpace
    translationSpace=BCS_BoneSpace
    alphaBoolBlend {blendInTime=0.2, blendOption=Linear, blendOutTime=0.2}
    boneToModify {boneName=C_Tail_A_1_Jnt}
    rotation {pitch=15}
   }
   n6 = "Transform (Modify) Bone - Bone: C_Tail_A_2_Jnt"!(ComponentPose=@n5, bAlphaBoolEnabled:bool=True, Rotation="-8,0,0") -> (Pose) {
    alphaInputType=Bool
    rotationMode=BMM_Additive
    rotationSpace=BCS_BoneSpace
    translationSpace=BCS_BoneSpace
    alphaBoolBlend {blendInTime=0.2, blendOption=Linear, blendOutTime=0.2}
    boneToModify {boneName=C_Tail_A_2_Jnt}
    rotation {pitch=-8}
   }
   n7 = "Transform (Modify) Bone - Bone: C_Tail_A_7_Jnt"!(ComponentPose=@n6, bAlphaBoolEnabled:bool=True, Rotation="20,0,0") -> (Pose) {
    alphaInputType=Bool
    rotationMode=BMM_Additive
    rotationSpace=BCS_BoneSpace
    translationSpace=BCS_BoneSpace
    alphaBoolBlend {blendInTime=0.2, blendOption=Linear, blendOutTime=0.2}
    boneToModify {boneName=C_Tail_A_7_Jnt}
    rotation {pitch=20}
    translation=<-0.631 3.384 0.129>
   }
   n8 = "Transform (Modify) Bone - Bone: C_Tail_A_8_Jnt"!(ComponentPose=@n7, bAlphaBoolEnabled:bool=True, Rotation="20,0,0") -> (Pose) {
    alphaInputType=Bool
    rotationMode=BMM_Additive
    rotationSpace=BCS_BoneSpace
    translationSpace=BCS_BoneSpace
    alphaBoolBlend {blendInTime=0.2, blendOption=Linear, blendOutTime=0.2}
    boneToModify {boneName=C_Tail_A_8_Jnt}
    rotation {pitch=20}
   }
   n9 = "Transform (Modify) Bone - Bone: C_Tail_A_9_Jnt"!(ComponentPose=@n8, bAlphaBoolEnabled:bool=True, Rotation="0,10,0") -> (Pose) {
    alphaInputType=Bool
    rotationMode=BMM_Additive
    rotationSpace=BCS_BoneSpace
    translationSpace=BCS_BoneSpace
    alphaBoolBlend {blendInTime=0.2, blendOption=Linear, blendOutTime=0.2}
    boneToModify {boneName=C_Tail_A_9_Jnt}
    rotation {yaw=10}
   }
   n10 = "Transform (Modify) Bone - Bone: C_Tail_A_Base_Jnt"!(ComponentPose=@n9, bAlphaBoolEnabled:bool=True, Translation="0,0,0", Rotation="5,0,0") -> (Pose) {
    alphaInputType=Bool
    rotationMode=BMM_Additive
    rotationSpace=BCS_BoneSpace
    scaleSpace=BCS_BoneSpace
    translationMode=BMM_Replace
    translationSpace=BCS_BoneSpace
    alphaBoolBlend {blendInTime=0.2, blendOption=Linear, blendOutTime=0.2}
    boneToModify {boneName=C_Tail_A_Base_Jnt}
    rotation {pitch=5}
   }
   n11 = "Transform (Modify) Bone - Bone: C_Tail_A_2_Jnt"!(ComponentPose=@n10, bAlphaBoolEnabled=True, Translation:Vector="0,0,0", Rotation="15,0,0") -> (Pose) {
    alphaInputType=Bool
    rotationMode=BMM_Additive
    rotationSpace=BCS_BoneSpace
    scaleSpace=BCS_BoneSpace
    translationMode=BMM_Additive
    translationSpace=BCS_BoneSpace
    alphaBoolBlend {blendInTime=0.2, blendOption=Linear, blendOutTime=0.2}
    boneToModify {boneName=C_Tail_A_2_Jnt}
    rotation {pitch=15}
   }
   n12 = "Transform (Modify) Bone - Bone: C_Tail_A_3_Jnt"!(ComponentPose=@n11, bAlphaBoolEnabled:bool=True, Rotation="20,0,0") -> (Pose) {
    alphaInputType=Bool
    rotationMode=BMM_Additive
    rotationSpace=BCS_BoneSpace
    scaleSpace=BCS_BoneSpace
    translationSpace=BCS_BoneSpace
    alphaBoolBlend {blendInTime=0.2, blendOption=Linear, blendOutTime=0.2}
    boneToModify {boneName=C_Tail_A_3_Jnt}
    rotation {pitch=20}
   }
   n13 = "Transform (Modify) Bone - Bone: C_Tail_A_4_Jnt"!(ComponentPose=@n12, bAlphaBoolEnabled:bool=True, Rotation="15,0,0") -> (Pose) {
    alphaInputType=Bool
    rotationMode=BMM_Additive
    rotationSpace=BCS_BoneSpace
    scaleSpace=BCS_BoneSpace
    translationSpace=BCS_BoneSpace
    alphaBoolBlend {blendInTime=0.2, blendOption=Linear, blendOutTime=0.2}
    boneToModify {boneName=C_Tail_A_4_Jnt}
    rotation {pitch=15}
   }
   n14 = "Transform (Modify) Bone - Bone: C_Tail_A_5_Jnt"!(ComponentPose=@n13, bAlphaBoolEnabled:bool=True, Rotation="15,0,0") -> (Pose) {
    alphaInputType=Bool
    rotationMode=BMM_Additive
    rotationSpace=BCS_BoneSpace
    scaleSpace=BCS_BoneSpace
    translationSpace=BCS_BoneSpace
    alphaBoolBlend {blendInTime=0.2, blendOption=Linear, blendOutTime=0.2}
    boneToModify {boneName=C_Tail_A_5_Jnt}
    rotation {pitch=15}
   }
   n15 = "Transform (Modify) Bone - Bone: C_Tail_A_1_Jnt"(ComponentPose=@n14, bAlphaBoolEnabled=True, Rotation="90,0,0") {
    alphaInputType=Bool
    rotationMode=BMM_Additive
    rotationSpace=BCS_BoneSpace
    alphaBoolBlend {blendOption=Linear}
    boneToModify {boneName=C_Tail_A_1_Jnt}
    rotation {pitch=90}
   }
   n16 = "Transform (Modify) Bone - Bone: C_Tail_A_2_Jnt"!(ComponentPose=@n15, bAlphaBoolEnabled=True, Rotation="60,0,0") -> (Pose) {
    alphaInputType=Bool
    rotationMode=BMM_Additive
    rotationSpace=BCS_BoneSpace
    alphaBoolBlend {blendOption=Linear}
    boneToModify {boneName=C_Tail_A_2_Jnt}
    rotation {pitch=60}
   }
   n17 = "Transform (Modify) Bone - Bone: C_Tail_A_3_Jnt"(ComponentPose=@n16, bAlphaBoolEnabled=True, Rotation="-20,0,0") {
    alphaInputType=Bool
    rotationMode=BMM_Additive
    rotationSpace=BCS_BoneSpace
    alphaBoolBlend {blendOption=Linear}
    boneToModify {boneName=C_Tail_A_3_Jnt}
    rotation {pitch=-20}
   }
   n18 = "Transform (Modify) Bone - Bone: C_Tail_A_5_Jnt"(ComponentPose=@n17, bAlphaBoolEnabled=True, Rotation="-20,0,0") {
    alphaInputType=Bool
    rotationMode=BMM_Additive
    rotationSpace=BCS_BoneSpace
    alphaBoolBlend {blendOption=Linear}
    boneToModify {boneName=C_Tail_A_5_Jnt}
    rotation {pitch=-20}
   }
   n19 = "Transform (Modify) Bone - Bone: C_Tail_A_6_Jnt"!(ComponentPose=@n18, bAlphaBoolEnabled:bool=True, Rotation="-20,0,0") -> (Pose) {
    alphaInputType=Bool
    rotationMode=BMM_Additive
    rotationSpace=BCS_BoneSpace
    alphaBoolBlend {blendOption=Linear}
    boneToModify {boneName=C_Tail_A_6_Jnt}
    rotation {pitch=-20}
   }
   n20 = "Transform (Modify) Bone - Bone: C_Tail_A_7_Jnt"(ComponentPose=@n19, bAlphaBoolEnabled=True, Rotation="-20,0,0") {
    alphaInputType=Bool
    rotationMode=BMM_Additive
    rotationSpace=BCS_BoneSpace
    alphaBoolBlend {blendOption=Linear}
    boneToModify {boneName=C_Tail_A_7_Jnt}
    rotation {pitch=-20}
   }
   n21 = "Transform (Modify) Bone - Bone: C_Tail_A_8_Jnt"(ComponentPose=@n20, bAlphaBoolEnabled=True, Rotation="-20,0,0") {
    alphaInputType=Bool
    rotationMode=BMM_Additive
    rotationSpace=BCS_BoneSpace
    alphaBoolBlend {blendOption=Linear}
    boneToModify {boneName=C_Tail_A_8_Jnt}
    rotation {pitch=-20}
   }
   n22 = "Transform (Modify) Bone - Bone: C_Tail_A_9_Jnt"(ComponentPose=@n21, bAlphaBoolEnabled=True, Rotation="-15,0,0") {
    alphaInputType=Bool
    rotationMode=BMM_Additive
    rotationSpace=BCS_BoneSpace
    alphaBoolBlend {blendOption=Linear}
    boneToModify {boneName=C_Tail_A_9_Jnt}
    rotation {pitch=-15}
   }
   n23 = "Transform (Modify) Bone - Bone: C_Tail_A_1_Jnt"(ComponentPose=@n22, bAlphaBoolEnabled=True, Rotation="-10,0,0") {
    alphaInputType=Bool
    rotationMode=BMM_Additive
    rotationSpace=BCS_BoneSpace
    translationSpace=BCS_BoneSpace
    alphaBoolBlend {blendInTime=1, blendOption=Linear, blendOutTime=5}
    boneToModify {boneName=C_Tail_A_1_Jnt}
    rotation {pitch=-10}
   }
   n24 = "Transform (Modify) Bone - Bone: C_Tail_A_2_Jnt"(ComponentPose=@n23, bAlphaBoolEnabled=True, Rotation="10,0,0") {
    alphaInputType=Bool
    rotationMode=BMM_Additive
    rotationSpace=BCS_BoneSpace
    translationSpace=BCS_BoneSpace
    alphaBoolBlend {blendInTime=1, blendOption=Linear, blendOutTime=5}
    boneToModify {boneName=C_Tail_A_2_Jnt}
    rotation {pitch=10}
   }
   n25 = "Transform (Modify) Bone - Bone: C_Tail_A_6_Jnt"(ComponentPose=@n24, bAlphaBoolEnabled=True, Rotation="15,0,0") {
    alphaInputType=Bool
    rotationMode=BMM_Additive
    rotationSpace=BCS_BoneSpace
    translationSpace=BCS_BoneSpace
    alphaBoolBlend {blendInTime=1, blendOption=Linear, blendOutTime=5}
    boneToModify {boneName=C_Tail_A_6_Jnt}
    rotation {pitch=15}
   }
   n26 = "Transform (Modify) Bone - Bone: C_Tail_A_7_Jnt"(ComponentPose=@n25, bAlphaBoolEnabled=True, Rotation="10,0,0") {
    alphaInputType=Bool
    rotationMode=BMM_Additive
    rotationSpace=BCS_BoneSpace
    translationSpace=BCS_BoneSpace
    alphaBoolBlend {blendInTime=1, blendOption=Linear, blendOutTime=5}
    boneToModify {boneName=C_Tail_A_7_Jnt}
    rotation {pitch=10}
    translation=<-0.631 3.384 0.129>
   }
   n27 = "Transform (Modify) Bone - Bone: C_Tail_A_8_Jnt"(ComponentPose=@n26, bAlphaBoolEnabled=True, Rotation="-10,0,0") {
    alphaInputType=Bool
    rotationMode=BMM_Additive
    rotationSpace=BCS_BoneSpace
    translationSpace=BCS_BoneSpace
    alphaBoolBlend {blendInTime=1, blendOption=Linear, blendOutTime=5}
    boneToModify {boneName=C_Tail_A_8_Jnt}
    rotation {pitch=-10}
   }
   Output Pose(Result=Component To Local(ComponentPose=@n27))
   Get isWearingCape() -> (isWearingCape:bool)
  ]}
 EventGraph {nodes [
   Event OnCharacterCustomizationLoaded()
   > Set isWearingBackbling(isWearingBackbling=false)
   > Set isWearingCape(isWearingCape=CheckForBackbling().bIsBackblingLong)
   > Branch(Condition=OR Boolean(A=Get isWearingCape(), B=Get isWearingBackbling()))
   > n1 = Get Dynamic Material Instance(SkeletalMeshComponent=GetPartSkeletalMeshComponent())
   > SetScalarParameterValue(self=@n1, ParameterName=Hide Element 1, Value=1)
   Event BlueprintUpdateAnimation()
   > Sequence()
   n2 = Get bIsMoving2D()
   > Set isOnlyCrouching(isOnlyCrouching:bool=AND Boolean(A=Get bIsCrouching(), B=NOT Boolean(A=@n2))) -> (Output_Get:bool)
   n3 = Get bIsMoving2D()
   n4 = Get isOnlyCrouching() -> (isOnlyCrouching:bool)
   > Set isNotMoving(isNotMoving:bool=AND Boolean(A=NOT Boolean(A=@n3), B=NOT Boolean(A=@n4))) -> (Output_Get:bool)
   TryGetPawnOwner()
  ]}
]
variables [
 Is Wearing Backbling {type=boolean}
 isNotMoving {type=boolean}
 isOnlyCrouching {type=boolean}
 isWearingBackbling {type=boolean}
 isWearingCape {type=boolean}
 tailBase_Rotation {type=Rotator}
 y movement {type=number}
]
